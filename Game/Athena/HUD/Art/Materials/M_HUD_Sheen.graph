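# UI-domain HUD material that produces a sheen effect, with static switches choosing the style.
# UseRadialAmbientSheen draws a soft line on continuously rotating UVs, multiplied by an exponent-shaped edge frame built from centered UVs.
# Otherwise, it samples a clamped shine texture with slightly rotated UVs, slid diagonally by the Position parameter.
# UseColor turns that texture into a blend between ColorA and ColorB.
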Material /Game/Athena/HUD/Art/Materials/M_HUD_Sheen
graph {
 properties {materialDomain=MD_UI, numCustomizedUVs=2, pixelDepthOffsetMode=PDOM_Legacy, usedShadingModels=Default Lit}
 nodes [
  ScalarParameter() {defaultValue=1, parameterName=Opacity}
  n1 = TextureCoordinate()
  n2 = Rotator(Coordinate=@n1) {speed=-0.5}
  n3 = Constant2Vector() {g=-0.5, r=0.5}
  n4 = Constant2Vector() {g=1.5, r=0.5}
  n5 = ScalarParameter() {defaultValue=0.5, parameterName=RadialSheenLineWidth}
  n6 = ScalarParameter() {defaultValue=0.1, parameterName=RadialSheenEdgeFalloff}
  n7 = ConstantBiasScale(None=@n1) {bias=-0.5, scale=2}
  n8 = Power(
   Base=LinearInterpolate(B=Abs(None=@n7), Alpha=Constant() {r=1}) {constA=1}
   Exp=ScalarParameter() {defaultValue=3.25, parameterName=FrameExponent}
  )
  n9 = ComponentMask(None=@n8) {r=true}
  n10 = ComponentMask(None=@n8) {g=true}
  n11 = VectorParameter() {parameterName=ColorA, defaultValue=<rgba 0 0.87 1 1>}
  n12 = VectorParameter() {parameterName=ColorB, defaultValue=<rgba 1 1 1 1>}
  n13 = TextureCoordinate()
  n14 = Constant() {r=0.05}
  n15 = LinearInterpolate(Alpha=ScalarParameter() {parameterName=Position}) {constA=1, constAlpha=0.2, constB=-1}
  n16 = TextureSample(UVs=Add(
    A=MaterialFunctionCall(UVs=@n13, "Rotation Angle (0-1)"=@n14) {materialFunction=&CustomRotator}.Rotated Values
    B=AppendVector(A=@n15, B=@n15)
   )) {samplerSource=SSM_Clamp_WorldGroupSettings, texture=/Game/UI/Foundation/Textures/Pattern/T-UI-Shine}
  StaticSwitchParameter(
   True=Multiply(
    A=MaterialFunctionCall!(Coordinates=@n2, P0=@n3, P1=@n4, Line Width=@n5, Edge Falloff=@n6, Square Corners?) {materialFunction=/Engine/Functions/Engine_MaterialFunctions01/Coordinates/DrawLine-2D}.Result
    B=Add(A=@n9, B=@n10)
   )
   False=StaticSwitchParameter(True=LinearInterpolate(A=@n11.RGB, B=@n12.RGB, Alpha=@n16.RGB), False=@n16.RGB) {parameterName=UseColor}
  ) {parameterName=UseRadialAmbientSheen}
 ]
}
