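# HUD widget that tells the player who the squad's skydive leader is, or shows the key prompt to pass leadership.
# Construct binds to the player controller's skydive leader change and the keybinds change events, seeds the current state, and Destruct unbinds.
# Leader changes toggle the prompt, set the leader name and pick local or remote variants.
# Game phase steps control overall visibility and switch between announcement and follow/pass layouts.
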WidgetBlueprint /Game/Athena/HUD/FidgetSpinner/FidgetHUDWidget
graphs [EventGraph {nodes [
   n1 = SkydiveLeaderChanged!() -> (OutputDelegate, then, bEnabled:bool=~, bLocal:bool=~, SkydiveLeader:Fort Player State Object Reference) {is=title K2Node_CustomEvent}
   > SetVisibility(
    self=Get PhaseSwitcher() -> (PhaseSwitcher:Widget Switcher Object Reference)
    InVisibility=Select(Option 0=Collapsed, Option 1=SelfHitTestInvisible, Index=@n1.bEnabled)
   )
   n2 = Select(Option 0:int=0, Option 1:int=1, Index=@n1.bLocal) -> (ReturnValue:int)
   > SetActiveWidgetIndex(self=Get LocalRemoteSwitcherAPhase() -> (LocalRemoteSwitcherAPhase:Widget Switcher Object Reference), Index=@n2)
   > SetActiveWidgetIndex(self=Get LocalRemoteSwitcherWPhase() -> (LocalRemoteSwitcherWPhase:Widget Switcher Object Reference), Index=@n2)
   > Is Valid(InputObject=@n1.SkydiveLeader)
   > "SetText (Text)"(
    self=Get SkydiveLeaderName() -> (SkydiveLeaderName:Common Text Object Reference)
    InText="To Text (String)"(InString=GetPlayerName!(self:Player State Object Reference=@n1.SkydiveLeader) -> (ReturnValue:String))
   )
   KeybindsChanged() {is=title K2Node_CustomEvent}
   > SetBoundAction!(
    execute
    self:Fort Keybind Widget Object Reference=Get KeybindWidget_C_1() -> (KeybindWidget_C_1:Keybind Widget Object Reference)
    NewBoundAction:Name=GetUseActionName!(
     self:Fort Global UIContext Object Reference=Get FortGlobalUIContext() -> (ReturnValue:Fort Global UIContext Object Reference) {is=title K2Node_GetSubsystem}
     OverrideInputType:ECommonInputType Enum=Count
    ) -> (ReturnValue:Name=None)
   ) -> (then)
   Event Construct()
   > n3 = Cast To FortPlayerControllerAthena(Object=GetOwningPlayer())
   n4 = Create Event()
   > Bind Event to On Skydive Leader Changed(self:Fort Player Controller Athena Object Reference=@n3, Delegate=@n4) {is=title K2Node_AddDelegate}
   n5 = Create Event()
   n6 = Get FortGlobalUIContext()
   > Bind Event to On Keybinds Changed(self:Fort Global UIContext Object Reference=@n6, Delegate=@n5) {is=title K2Node_AddDelegate}
   > KeybindsChanged() {is=title K2Node_CallFunction}
   n7 = Get SkydiveLeader!(self:Fort Player Controller Athena Object Reference=@n3) -> (SkydiveLeader:Fort Player State Object Reference)
   > Is Valid(InputObject=@n7)
   > SkydiveLeaderChanged!(
    execute
    self
    bEnabled:bool=true
    bLocal:bool="Equal (Object)"(A=@n7, B=Get PlayerState!(self:Controller Object Reference=@n3) -> (PlayerState:Player State Object Reference)) {is=title K2Node_CallFunction}
    SkydiveLeader:Fort Player State Object Reference=@n7
   ) -> (then) {is=title K2Node_CallFunction}
   Event Destruct()
   > n8 = Cast To FortPlayerControllerAthena(Object=GetOwningPlayer())
   > Unbind Event from On Skydive Leader Changed(self:Fort Player Controller Athena Object Reference=@n8, Delegate=@n4) {is=title K2Node_RemoveDelegate}
   > Unbind Event from On Keybinds Changed(self:Fort Global UIContext Object Reference=@n6, Delegate=@n5) {is=title K2Node_RemoveDelegate}
   n9 = Event OnGamePhaseStepChanged!() -> (OutputDelegate, then, GamePhaseStep:EAthenaGamePhaseStep Enum=None) {is=title K2Node_Event}
   > SetVisibility(
    self=Get MainOverlay() -> (MainOverlay:Overlay Object Reference)
    InVisibility=Select!(
     None:ESlateVisibility Enum=Collapsed
     Setup:ESlateVisibility Enum=SelfHitTestInvisible
     Warmup:ESlateVisibility Enum=SelfHitTestInvisible
     GetReady:ESlateVisibility Enum=SelfHitTestInvisible
     BusLocked:ESlateVisibility Enum=SelfHitTestInvisible
     BusFlying:ESlateVisibility Enum=SelfHitTestInvisible
     StormForming:ESlateVisibility Enum=SelfHitTestInvisible
     StormHolding:ESlateVisibility Enum=SelfHitTestInvisible
     StormShrinking:ESlateVisibility Enum=SelfHitTestInvisible
     Countdown:ESlateVisibility Enum=Collapsed
     FinalCountdown:ESlateVisibility Enum=Collapsed
     EndGame:ESlateVisibility Enum=Collapsed
     Count:ESlateVisibility Enum=Collapsed
     Index:EAthenaGamePhaseStep Enum=@n9.GamePhaseStep
    ) -> (ReturnValue:ESlateVisibility Enum)
   )
   > SetActiveWidgetIndex(
    self=Get PhaseSwitcher()
    Index=Select(
     None:int=0
     Setup:int=0
     Warmup:int=0
     GetReady:int=0
     BusLocked:int=1
     BusFlying:int=1
     StormForming:int=1
     StormHolding:int=1
     StormShrinking:int=0
     Countdown:int=0
     FinalCountdown:int=0
     EndGame:int=0
     Count:int=0
     Index=@n9.GamePhaseStep
    ) -> (ReturnValue:int)
   )
   Event PreConstruct()
   Event Tick()
  ]}]
variables []
widgets {
 info {inheritedWidgetCount=1, parentClass=/Script/FortniteUI.AthenaZebulonDroneWidgetBase, rootWidgetClass=&Overlay, widgetCount=15}
 widgets [
  MainOverlay {widgetClassPath=&Overlay, bIsVariable=true, widget=MainOverlay}
  PhaseSwitcher {
   widgetClassPath=&WidgetSwitcher
   bIsVariable=true
   parent=MainOverlay
   widget=PhaseSwitcher
   properties {visibility=Collapsed}
   slot {class=OverlaySlot}
  }
  LocalRemoteSwitcherWPhase {
   widgetClassPath=&WidgetSwitcher
   bIsVariable=true
   parent=PhaseSwitcher
   widget=LocalRemoteSwitcherWPhase
   slot {class=WidgetSwitcherSlot}
  }
  HorizontalBox_0 {widgetClassPath=&HorizontalBox, parent=LocalRemoteSwitcherWPhase, widget=HorizontalBox_0, slot {class=WidgetSwitcherSlot}}
  SkydiveLeaderName {
   widgetClassPath=&CommonTextBlock
   bIsVariable=true
   parent=HorizontalBox_0
   widget=SkydiveLeaderName
   properties {
    text="{Player}"
    colorAndOpacity {colorUseRule=UseColor_Specified, specifiedColor=<rgba 1 0.7913 0.1356 1>}
    font {
     bForceMonospaced=false
     bMaterialIsStencil=false
     fontMaterial=null
     fontObject=&Burbank
     letterSpacing=0
     monospacedWidth=1
     size=24
     skewAmount=0
     typefaceFontName=Burbank-Black
     outlineSettings {
      bApplyOutlineToDropShadows=false
      bMiteredCorners=false
      bSeparateFillAlpha=false
      outlineMaterial=null
      outlineSize=1
      outlineColor=<rgba 0 0 0 0.15>
     }
    }
    shadowColorAndOpacity=<rgba 0.0052 0.0052 0.0056 0.822>
    shadowOffset=<3 3>
   }
   slot {class=HorizontalBoxSlot, properties {horizontalAlignment=HAlign_Center, verticalAlignment=VAlign_Center, padding=<ltrb 4 0 4 0>}}
  }
  TextBlock_3 {
   widgetClassPath=&TextBlock
   parent=HorizontalBox_0
   widget=TextBlock_3
   properties {
    text=IS YOUR SKYDIVE LEADER
    font {
     bForceMonospaced=false
     bMaterialIsStencil=false
     fontMaterial=null
     fontObject=&NotoSans
     letterSpacing=0
     monospacedWidth=1
     size=18
     skewAmount=0
     typefaceFontName=NotoSans-Bold
     outlineSettings {
      bApplyOutlineToDropShadows=true
      bMiteredCorners=false
      bSeparateFillAlpha=false
      outlineMaterial=/Game/Athena/HUD/Art/Materials/M_HUD_Sheen
      outlineSize=1
      outlineColor=<rgba 1 1 1 1>
     }
    }
    shadowOffset=<0 2>
   }
   slot {class=HorizontalBoxSlot, properties {horizontalAlignment=HAlign_Left, verticalAlignment=VAlign_Center}}
  }
  HorizontalBox_1 {widgetClassPath=&HorizontalBox, parent=LocalRemoteSwitcherWPhase, widget=HorizontalBox_1, slot {class=WidgetSwitcherSlot}}
  TextBlock_4 {
   widgetClassPath=&TextBlock
   parent=HorizontalBox_1
   widget=TextBlock_4
   properties {
    text=YOU ARE THE SKYDIVE LEADER
    font {
     bForceMonospaced=false
     bMaterialIsStencil=false
     fontMaterial=null
     fontObject=&NotoSans
     letterSpacing=0
     monospacedWidth=1
     size=18
     skewAmount=0
     typefaceFontName=NotoSans-Bold
     outlineSettings {
      bApplyOutlineToDropShadows=true
      bMiteredCorners=false
      bSeparateFillAlpha=false
      outlineMaterial=&M_HUD_Sheen
      outlineSize=1
      outlineColor=<rgba 1 1 1 1>
     }
    }
    shadowOffset=<0 2>
   }
   slot {class=HorizontalBoxSlot, properties {horizontalAlignment=HAlign_Left, verticalAlignment=VAlign_Center}}
  }
  LocalRemoteSwitcherAPhase {
   widgetClassPath=&WidgetSwitcher
   bIsVariable=true
   parent=PhaseSwitcher
   widget=LocalRemoteSwitcherAPhase
   slot {class=WidgetSwitcherSlot}
  }
  DroneRemote {widgetClassPath=&HorizontalBox, parent=LocalRemoteSwitcherAPhase, widget=DroneRemote, slot {class=WidgetSwitcherSlot}}
  FeedbackText {
   widgetClassPath=&TextBlock
   parent=DroneRemote
   widget=FeedbackText
   properties {
    text=FOLLOWING SKYDIVE LEADER
    font {
     bForceMonospaced=false
     bMaterialIsStencil=false
     fontMaterial=null
     fontObject=&NotoSans
     letterSpacing=0
     monospacedWidth=1
     size=18
     skewAmount=0
     typefaceFontName=NotoSans-Bold
     outlineSettings {
      bApplyOutlineToDropShadows=true
      bMiteredCorners=false
      bSeparateFillAlpha=false
      outlineMaterial=&M_HUD_Sheen
      outlineSize=1
      outlineColor=<rgba 1 1 1 1>
     }
    }
    shadowOffset=<0 2>
   }
   slot {class=HorizontalBoxSlot, properties {horizontalAlignment=HAlign_Left, verticalAlignment=VAlign_Center}}
  }
  DroneLocal {widgetClassPath=&HorizontalBox, parent=LocalRemoteSwitcherAPhase, widget=DroneLocal, slot {class=WidgetSwitcherSlot}}
  KeybindWidget_C_1 {
   widgetClassPath=/FortUILibrary/Input/KeybindWidget
   bIsVariable=true
   parent=DroneLocal
   widget=KeybindWidget_C_1
   properties {boundAction=Use}
   slot {class=HorizontalBoxSlot, properties {verticalAlignment=VAlign_Center, padding=<ltrb 4 0 4 0>}}
  }
  TextBlock_1 {
   widgetClassPath=&TextBlock
   parent=DroneLocal
   widget=TextBlock_1
   properties {
    text=TO PASS SKYDIVE LEADER
    font {
     bForceMonospaced=false
     bMaterialIsStencil=false
     fontMaterial=null
     fontObject=&NotoSans
     letterSpacing=0
     monospacedWidth=1
     size=18
     skewAmount=0
     typefaceFontName=NotoSans-Bold
     outlineSettings {
      bApplyOutlineToDropShadows=true
      bMiteredCorners=false
      bSeparateFillAlpha=false
      outlineMaterial=&M_HUD_Sheen
      outlineSize=1
      outlineColor=<rgba 1 1 1 1>
     }
    }
    shadowOffset=<0 2>
   }
   slot {class=HorizontalBoxSlot, properties {horizontalAlignment=HAlign_Left, verticalAlignment=VAlign_Center}}
  }
  TutorialHighlight {widgetClassPath=&TutorialHighlight}
 ]
}
